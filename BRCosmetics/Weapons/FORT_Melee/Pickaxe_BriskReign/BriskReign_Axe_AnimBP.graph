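# Animation Blueprint for the BriskReign pickaxe that simulates its four-bone tassel with chained AnimDynamics nodes on the cached input pose.
# In the front end, bIsOnPawn selects an angular-spring chain when true and a world-space chain when false.
# Outside it, a pitch offset feeds a handle-relative chain.
# The update event sets gravity and rotation variables by jump/fall state, though no shown node reads them.
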AnimBlueprint /BRCosmetics/Weapons/FORT_Melee/Pickaxe_BriskReign/BriskReign_Axe_AnimBP
target_skeleton=&/BriskReign_Axe_Skeleton
graphs [
 AnimGraph {nodes [
   n1 = Anim Dynamics - Bone: dyn_tassel_1(ComponentPose=Local To Component(LocalPose=Use cached pose 'BASE INPUT'()), Alpha=1, ExternalForce="0,0,0") {
    angularDampingOverride=0.9999
    angularSpringConstant=100
    bAngularSpring=true
    bOverrideAngularDamping=true
    bUseGravityOverride=true
    bUsePlanarLimit=false
    alphaBoolBlend {blendOption=Linear}
    boundBone {boneName=dyn_tassel_1}
    componentAppliedLinearAccClamp=<3500 3500 3500>
    componentLinearAccScale=<0.4 0.4 0.4>
    componentLinearVelScale=<0.6 0.6 0.6>
    gravityOverride {z=-980}
    physicsBodyDefinitions [{
      boundBone {boneName=dyn_tassel_1}
      boxExtents=<12 12 12>
      constraintSetup {angularLimitsMax=<15 120 30>, angularLimitsMin=<-15 -30 -30>, angularTarget=<1 0 0>}
      localJointOffset=<1.5 0 0>
     }]
    relativeSpaceBone {boneName=Melee_TwoHanded_Handle}
    simSpaceSettings {maxAngularAcceleration=70, maxAngularVelocity=10, simSpaceAngularAlpha=0.2}
   }
   n2 = Anim Dynamics - Bone: dyn_tassel_2(ComponentPose=@n1, Alpha=1, ExternalForce="0,0,0") {
    angularDampingOverride=0.9999
    angularSpringConstant=250
    bAngularSpring=true
    bOverrideAngularDamping=true
    bUseGravityOverride=true
    bUsePlanarLimit=false
    alphaBoolBlend {blendOption=Linear}
    boundBone {boneName=dyn_tassel_2}
    componentAppliedLinearAccClamp=<3500 3500 3500>
    componentLinearAccScale=<0.4 0.4 0.4>
    componentLinearVelScale=<0.6 0.6 0.6>
    gravityOverride {z=-980}
    physicsBodyDefinitions [{
      boundBone {boneName=dyn_tassel_2}
      boxExtents=<11 11 11>
      constraintSetup {angularLimitsMax=<30 30 30>, angularLimitsMin=<-30 -30 -30>, angularTarget=<1 0 0>}
      localJointOffset=<1.5 0 0>
     }]
    relativeSpaceBone {boneName=Melee_TwoHanded_Handle}
    simSpaceSettings {maxAngularAcceleration=70, maxAngularVelocity=10, simSpaceAngularAlpha=0.5}
    sphericalLimits [{limitRadius=10, drivingBone {boneName=Melee_TwoHanded_Handle}, sphereLocalOffset=<5 0 3>}]
   }
   n3 = Anim Dynamics - Bone: dyn_tassel_3(ComponentPose=@n2, Alpha=1, ExternalForce="0,0,0") {
    angularDampingOverride=0.9999
    angularSpringConstant=250
    bAngularSpring=true
    bOverrideAngularDamping=true
    bUseGravityOverride=true
    bUsePlanarLimit=false
    alphaBoolBlend {blendOption=Linear}
    boundBone {boneName=dyn_tassel_3}
    componentAppliedLinearAccClamp=<3500 3500 3500>
    componentLinearAccScale=<0.4 0.4 0.4>
    componentLinearVelScale=<0.6 0.6 0.6>
    gravityOverride {z=-980}
    physicsBodyDefinitions [{
      boundBone {boneName=dyn_tassel_3}
      boxExtents=<10 10 10>
      constraintSetup {angularLimitsMax=<30 30 30>, angularLimitsMin=<-30 -30 -30>, angularTarget=<1 0 0>}
      localJointOffset=<1.5 0 0>
     }]
    relativeSpaceBone {boneName=Melee_TwoHanded_Handle}
    simSpaceSettings {maxAngularAcceleration=70, maxAngularVelocity=10, simSpaceAngularAlpha=0.5}
    sphericalLimits [{limitRadius=10, drivingBone {boneName=Melee_TwoHanded_Handle}, sphereLocalOffset=<5 0 3>}]
   }
   n4 = Anim Dynamics - Bone: dyn_tassel_4(ComponentPose=@n3, Alpha=1, ExternalForce="0,0,0") {
    angularDampingOverride=0.9999
    angularSpringConstant=250
    bAngularSpring=true
    bOverrideAngularDamping=true
    bUseGravityOverride=true
    bUsePlanarLimit=false
    alphaBoolBlend {blendOption=Linear}
    boundBone {boneName=dyn_tassel_4}
    componentAppliedLinearAccClamp=<3500 3500 3500>
    componentLinearAccScale=<0.4 0.4 0.4>
    componentLinearVelScale=<0.6 0.6 0.6>
    gravityOverride {z=-980}
    physicsBodyDefinitions [{
      boundBone {boneName=dyn_tassel_4}
      boxExtents=<9 9 9>
      constraintSetup {angularLimitsMax=<30 30 30>, angularLimitsMin=<-30 -30 -30>, angularTarget=<1 0 0>}
      localJointOffset=<1.5 0 0>
     }]
    relativeSpaceBone {boneName=Melee_TwoHanded_Handle}
    simSpaceSettings {maxAngularAcceleration=70, maxAngularVelocity=10, simSpaceAngularAlpha=0.5}
    sphericalLimits [{limitRadius=10, drivingBone {boneName=Melee_TwoHanded_Handle}, sphereLocalOffset=<5 0 3>}]
   }
   n5 = Component To Local(ComponentPose=@n4)
   n6 = Anim Dynamics - Bone: dyn_tassel_1!(ComponentPose=Local To Component(LocalPose=Use cached pose 'BASE INPUT'()), Alpha=1, GravityOverride:Vector="0,0,0", ExternalForce="0,0,0") -> (Pose) {
    angularDampingOverride=1
    bOverrideAngularDamping=true
    bUsePlanarLimit=false
    simulationSpace=World
    alphaBoolBlend {blendOption=Linear}
    boundBone {boneName=dyn_tassel_1}
    componentAppliedLinearAccClamp=<3500 3500 3500>
    componentLinearAccScale=<0.2 0.2 0.2>
    componentLinearVelScale=<0.2 0.2 0.2>
    gravityOverride {z=-980}
    physicsBodyDefinitions [{
      boundBone {boneName=dyn_tassel_1}
      boxExtents=<22 22 22>
      constraintSetup {angularLimitsMax=<15 120 50>, angularLimitsMin=<-15 -30 -50>, angularTarget=<0 0 0>}
      localJointOffset=<1 0 0>
     }]
    relativeSpaceBone {boneName=Melee_TwoHanded_Handle}
   }
   n7 = Anim Dynamics - Bone: dyn_tassel_3!(ComponentPose=@n6, Alpha=1, GravityOverride:Vector="0,0,0", ExternalForce="0,0,0") -> (Pose) {
    angularDampingOverride=1
    bOverrideAngularDamping=true
    bUsePlanarLimit=false
    simulationSpace=World
    alphaBoolBlend {blendOption=Linear}
    boundBone {boneName=dyn_tassel_3}
    componentAppliedLinearAccClamp=<3500 3500 3500>
    componentLinearAccScale=<0.2 0.2 0.2>
    componentLinearVelScale=<0.2 0.2 0.2>
    gravityOverride {z=-980}
    physicsBodyDefinitions [{
      boundBone {boneName=dyn_tassel_3}
      boxExtents=<22 22 22>
      constraintSetup {angularLimitsMax=<80 30 30>, angularLimitsMin=<-80 -30 -30>, angularTarget=<0 0 0>}
      localJointOffset=<1 0 0>
     }]
    relativeSpaceBone {boneName=Melee_TwoHanded_Handle}
    sphericalLimits [{limitRadius=10, drivingBone {boneName=Melee_TwoHanded_Handle}, sphereLocalOffset=<5 0 3>}]
   }
   n8 = Anim Dynamics - Bone: dyn_tassel_2!(ComponentPose=@n7, Alpha=1, GravityOverride:Vector="0,0,0", ExternalForce="0,0,0") -> (Pose) {
    angularDampingOverride=1
    bOverrideAngularDamping=true
    bUsePlanarLimit=false
    simulationSpace=World
    alphaBoolBlend {blendOption=Linear}
    boundBone {boneName=dyn_tassel_2}
    componentAppliedLinearAccClamp=<3500 3500 3500>
    componentLinearAccScale=<0.2 0.2 0.2>
    componentLinearVelScale=<0.2 0.2 0.2>
    gravityOverride {z=-980}
    physicsBodyDefinitions [{
      boundBone {boneName=dyn_tassel_2}
      boxExtents=<22 22 22>
      constraintSetup {angularLimitsMax=<80 30 30>, angularLimitsMin=<-80 -30 -30>, angularTarget=<0 0 0>}
      localJointOffset=<1 0 0>
     }]
    relativeSpaceBone {boneName=Melee_TwoHanded_Handle}
    sphericalLimits [{limitRadius=10, drivingBone {boneName=Melee_TwoHanded_Handle}, sphereLocalOffset=<5 0 3>}]
   }
   n9 = Anim Dynamics - Bone: dyn_tassel_4!(ComponentPose=@n8, Alpha=1, GravityOverride:Vector="0,0,0", ExternalForce="0,0,0") -> (Pose) {
    angularDampingOverride=1
    bOverrideAngularDamping=true
    bUsePlanarLimit=false
    simulationSpace=World
    alphaBoolBlend {blendOption=Linear}
    boundBone {boneName=dyn_tassel_4}
    componentAppliedLinearAccClamp=<3500 3500 3500>
    componentLinearAccScale=<0.2 0.2 0.2>
    componentLinearVelScale=<0.2 0.2 0.2>
    gravityOverride {z=-980}
    physicsBodyDefinitions [{
      boundBone {boneName=dyn_tassel_4}
      boxExtents=<22 22 22>
      constraintSetup {angularLimitsMax=<80 30 30>, angularLimitsMin=<-80 -30 -30>, angularTarget=<0 0 0>}
      localJointOffset=<1 0 0>
     }]
    relativeSpaceBone {boneName=Melee_TwoHanded_Handle}
    sphericalLimits [{limitRadius=10, drivingBone {boneName=Melee_TwoHanded_Handle}, sphereLocalOffset=<5 0 3>}]
   }
   n10 = Component To Local(ComponentPose=@n9)
   n11 = Get bIsOnPawn() -> (bIsOnPawn:bool)
   n12 = "Transform (Modify) Bone - Bone: dyn_tassel_1"(ComponentPose=Local To Component(LocalPose=Use cached pose 'BASE INPUT'()), Alpha=1, Translation="0,0,0", Rotation="-70,0,0", Scale="1,1,1") {
    rotationMode=BMM_Additive
    rotationSpace=BCS_BoneSpace
    translationMode=BMM_Additive
    translationSpace=BCS_BoneSpace
    alphaBoolBlend {blendOption=Linear}
    boneToModify {boneName=dyn_tassel_1}
    rotation {pitch=-70}
   }
   n13 = Anim Dynamics - Bone: dyn_tassel_1(ComponentPose=@n12, Alpha=1, GravityOverride="0,0,0", ExternalForce="0,0,0") {
    angularDampingOverride=0.9999
    bOverrideAngularDamping=true
    bUseGravityOverride=true
    bUsePlanarLimit=false
    simulationSpace=BoneRelative
    alphaBoolBlend {blendOption=Linear}
    boundBone {boneName=dyn_tassel_1}
    componentAppliedLinearAccClamp=<3500 3500 3500>
    componentLinearAccScale=<0.2 0.2 0.2>
    componentLinearVelScale=<0.2 0.2 0.2>
    gravityOverride {z=-980}
    physicsBodyDefinitions [{
      boundBone {boneName=dyn_tassel_1}
      boxExtents=<10 10 10>
      constraintSetup {angularLimitsMax=<5 55 5>, angularLimitsMin=<-5 -80 -5>, angularTarget=<0 0 0>}
      localJointOffset=<1.5 0 0>
     }]
    relativeSpaceBone {boneName=Melee_TwoHanded_Handle}
   }
   n14 = Anim Dynamics - Bone: dyn_tassel_3(ComponentPose=@n13, Alpha=1, GravityOverride="0,0,0", ExternalForce="0,0,0") {
    angularDampingOverride=0.9999
    bOverrideAngularDamping=true
    bUseGravityOverride=true
    bUsePlanarLimit=false
    simulationSpace=BoneRelative
    alphaBoolBlend {blendOption=Linear}
    boundBone {boneName=dyn_tassel_3}
    componentAppliedLinearAccClamp=<3500 3500 3500>
    componentLinearAccScale=<0.2 0.2 0.2>
    componentLinearVelScale=<0.2 0.2 0.2>
    gravityOverride {z=-980}
    physicsBodyDefinitions [{
      boundBone {boneName=dyn_tassel_3}
      boxExtents=<10 10 10>
      constraintSetup {angularLimitsMax=<5 30 30>, angularLimitsMin=<-5 -30 -30>, angularTarget=<0 0 0>}
      localJointOffset=<1 0 0>
     }]
    relativeSpaceBone {boneName=Melee_TwoHanded_Handle}
    sphericalLimits [{limitRadius=10, drivingBone {boneName=Melee_TwoHanded_Handle}, sphereLocalOffset=<5 0 3>}]
   }
   n15 = Anim Dynamics - Bone: dyn_tassel_2(ComponentPose=@n14, Alpha=1, GravityOverride="0,0,0", ExternalForce="0,0,0") {
    angularDampingOverride=0.9999
    bOverrideAngularDamping=true
    bUseGravityOverride=true
    bUsePlanarLimit=false
    simulationSpace=BoneRelative
    alphaBoolBlend {blendOption=Linear}
    boundBone {boneName=dyn_tassel_2}
    componentAppliedLinearAccClamp=<3500 3500 3500>
    componentLinearAccScale=<0.2 0.2 0.2>
    componentLinearVelScale=<0.2 0.2 0.2>
    gravityOverride {z=-980}
    physicsBodyDefinitions [{
      boundBone {boneName=dyn_tassel_2}
      boxExtents=<10 10 10>
      constraintSetup {angularLimitsMax=<5 30 30>, angularLimitsMin=<-5 -30 -30>, angularTarget=<0 0 0>}
      localJointOffset=<1 0 0>
     }]
    relativeSpaceBone {boneName=Melee_TwoHanded_Handle}
    sphericalLimits [{limitRadius=10, drivingBone {boneName=Melee_TwoHanded_Handle}, sphereLocalOffset=<5 0 3>}]
   }
   n16 = Anim Dynamics - Bone: dyn_tassel_4(ComponentPose=@n15, Alpha=1, GravityOverride="0,0,0", ExternalForce="0,0,0") {
    angularDampingOverride=0.9999
    bOverrideAngularDamping=true
    bUseGravityOverride=true
    bUsePlanarLimit=false
    simulationSpace=BoneRelative
    alphaBoolBlend {blendOption=Linear}
    boundBone {boneName=dyn_tassel_4}
    componentAppliedLinearAccClamp=<3500 3500 3500>
    componentLinearAccScale=<0.2 0.2 0.2>
    componentLinearVelScale=<0.2 0.2 0.2>
    gravityOverride {z=-980}
    physicsBodyDefinitions [{
      boundBone {boneName=dyn_tassel_4}
      boxExtents=<10 10 10>
      constraintSetup {angularLimitsMax=<5 30 30>, angularLimitsMin=<-5 -30 -30>, angularTarget=<0 0 0>}
      localJointOffset=<1 0 0>
     }]
    relativeSpaceBone {boneName=Melee_TwoHanded_Handle}
    sphericalLimits [{limitRadius=10, drivingBone {boneName=Melee_TwoHanded_Handle}, sphereLocalOffset=<5 0 3>}]
   }
   Output Pose(Result=Blend Poses by bool(
     BlendPose_0=Blend Poses by bool(BlendPose_0=@n5, BlendPose_1=@n10, bActiveValue=@n11) {blendType=Linear}
     BlendPose_1=Component To Local(ComponentPose=@n16)
     bActiveValue=Get bIsInFrontEnd()
    ) {blendType=Linear}) {layerGroup=None}
   AnimGraphNode_SubInput_0 = Input Pose()
   Save cached pose 'BASE INPUT'(Pose=@AnimGraphNode_SubInput_0) {cacheName=BASE INPUT}
  ]}
 EventGraph {nodes [
   Event BlueprintUpdateAnimation()
   > Sequence()
   > n1 = Set IsJumpFalling(IsJumpFalling:bool=OR Boolean(A=Get bIsFalling(), B=Get bIsJumping())) -> (Output_Get:bool)
   > n2 = Branch(Condition=@n1)
   > Set Gravity Override(Gravity Override="0,0,400")
   > Set JumpFallingRotation(JumpFallingRotation:double=0) -> (Output_Get:double)
   Set Gravity Override(execute=@n2.else, Gravity Override="0,0,-980")
   > Set JumpFallingRotation(JumpFallingRotation=-90)
  ]}
]
variables [
 Gravity Override {type=Vector}
 IsJumpFalling {type=boolean}
 JumpFallingRotation {type=number}
 VelScale {type=Vector}
]
